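# Sound cue for the FxFm cosmetic pet's vocal reaction when the down-but-not-out (DBNO) state starts.
# A random node picks among four equally weighted waves, two DBNO clips and two reused Angry clips.
# All four quality-level branches share that same random node, so playback is identical at every quality setting.
# It uses close-range pet attenuation and the creature pets sound class, with a slight volume boost.
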SoundCue /BRCosmetics/Sounds/Pets/FxFm/PT_FxFm_DbnoStart_Cue
root=QualityLevel_0
nodes [
 {id=QualityLevel_0, children [Random_1,Random_1,Random_1,Random_1]}
 {id=Random_1, children [WavePlayer_0,WavePlayer_2,WavePlayer_5,WavePlayer_1], properties {weights [1,1,1,1]}}
 {id=WavePlayer_0, properties {soundWaveAssetPtr=&/PT_FxFm_DBNO_03}}
 {id=WavePlayer_1, properties {soundWaveAssetPtr=&/PT_FxFm_Angry_03}}
 {id=WavePlayer_2, properties {soundWaveAssetPtr=&/PT_FxFm_DBNO_02}}
 {id=WavePlayer_5, properties {soundWaveAssetPtr=&/PT_FxFm_Angry_02}}
]
properties {
 attenuationSettings=&Fort_Pets_Close_Athena_Att
 maxDistance=1800.0001
 soundClassObject=&Fort_Creatures_Pets
 volumeMultiplier=1.12
}
