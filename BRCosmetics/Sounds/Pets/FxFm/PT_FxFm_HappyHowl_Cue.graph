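# Sound cue for the happy howl of the FxFm pet.
# A Random node picks between two equally weighted howl waves and feeds a Modulator.
# All four inputs of the root quality-level node share that Modulator, so every quality level plays the same chain.
# It uses close-range pet attenuation and the creature pets sound class, with a slight volume boost.
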SoundCue /BRCosmetics/Sounds/Pets/FxFm/PT_FxFm_HappyHowl_Cue
root=QualityLevel_0
nodes [
 {id=Modulator_0, children [Random_0]}
 {id=QualityLevel_0, children [Modulator_0,Modulator_0,Modulator_0,Modulator_0]}
 {id=Random_0, children [WavePlayer_4,WavePlayer_0], properties {weights [1,1]}}
 {id=WavePlayer_0, properties {soundWaveAssetPtr=&/PT_FxFm_Happy_Howl_02}}
 {id=WavePlayer_4, properties {soundWaveAssetPtr=&/PT_FxFm_Happy_Howl_01}}
]
properties {
 attenuationSettings=&Fort_Pets_Close_Athena_Att
 maxDistance=1800.0001
 soundClassObject=&Fort_Creatures_Pets
 volumeMultiplier=1.12
}
