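# Sound cue for the FxFm pet's happy-start sound in BRCosmetics.
# A quality-level node routes every branch to the same random node, which picks among four equally weighted Happy wave variations.
# The cue uses the close-range pet attenuation asset and the creature pets sound class, with slightly reduced volume.
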SoundCue /BRCosmetics/Sounds/Pets/FxFm/PT_FxFm_HappyStart_Cue
root=QualityLevel_0
nodes [
 {id=QualityLevel_0, children [Random_1,Random_1,Random_1,Random_1]}
 {id=Random_1, children [WavePlayer_4,WavePlayer_5,WavePlayer_7,WavePlayer_8], properties {weights [1,1,1,1]}}
 {id=WavePlayer_4, properties {soundWaveAssetPtr=&/PT_FxFm_Happy_01}}
 {id=WavePlayer_5, properties {soundWaveAssetPtr=&/PT_FxFm_Happy_02}}
 {id=WavePlayer_7, properties {soundWaveAssetPtr=&/PT_FxFm_Happy_04}}
 {id=WavePlayer_8, properties {soundWaveAssetPtr=&/PT_FxFm_Happy_05}}
]
properties {
 attenuationSettings=&Fort_Pets_Close_Athena_Att
 maxDistance=1800.0001
 soundClassObject=&Fort_Creatures_Pets
 volumeMultiplier=0.896
}
