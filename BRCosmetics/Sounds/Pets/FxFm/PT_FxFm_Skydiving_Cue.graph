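# Sound cue for the FxFm pet's skydiving vocalization, using the pets sound class and close-range pet attenuation.
# A short randomized delay feeds a quality-level node whose branches all share one equal-weight random pick.
# The pick draws from three skydiving waves and one borrowed greet wave.
# Overridden concurrency allows one instance per owner and blocks new plays while it is active.
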SoundCue /BRCosmetics/Sounds/Pets/FxFm/PT_FxFm_Skydiving_Cue
root=Delay_0
nodes [
 {id=Delay_0, children [QualityLevel_0], properties {delayMax=0.3, delayMin=0.2}}
 {id=QualityLevel_0, children [Random_1,Random_1,Random_1,Random_1]}
 {id=Random_1, children [WavePlayer_0,WavePlayer_1,WavePlayer_7,WavePlayer_5], properties {weights [1,1,1,1]}}
 {id=WavePlayer_0, properties {soundWaveAssetPtr=&/PT_FxFm_Skydiving_01}}
 {id=WavePlayer_1, properties {soundWaveAssetPtr=&/PT_FxFm_Skydiving_02}}
 {id=WavePlayer_5, properties {soundWaveAssetPtr=&/PT_FxFm_Skydiving_04}}
 {id=WavePlayer_7, properties {soundWaveAssetPtr=&/PT_FxFm_Greet_01}}
]
properties {
 attenuationSettings=&Fort_Pets_Close_Athena_Att
 bOverrideConcurrency=true
 maxDistance=1800.0001
 soundClassObject=&Fort_Creatures_Pets
 volumeMultiplier=0.896
 concurrencyOverrides {bLimitToOwner=true, maxCount=1, resolutionRule=PreventNew}
}
